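; Gameplay ability for the Season 18 Birthday quest that tracks throwing a gift box, firing when activated by a gameplay event.
; On activation it calls the parent, copies the event's instigator tags and adds the Item.Consumable.GiftBox tag to them.
; It then sends a complex custom stat event with those tags to the owning player's quest manager and ends immediately.
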(event Ability|EventActivateAbilityFromEvent
  (params (EventData :type Gameplay%20Event%20Data%20Structure%20%28by%20ref%29))
  (bind (_eventtag _instigator _target _optionalobject _optionalobject2 _contexthandle _instigatortags) (Utilities|Struct|BreakGameplayEventData EventData))
  (|CallParentFunction EventData)
  (bind _output_get (Variables|Default|SetT_InstigatorTags _instigatortags))
  (GameplayTags|AddGameplayTag _output_get "(TagName=\"Item.Consumable.GiftBox\")" :!class /Script/GameplayTags.BlueprintGameplayTagLibrary)
  (Quest|SendComplexCustomStatEvent (Variables|GAVars|Data|GetOwningPlayerQuestManager) 0 (Variables|Default|GetT_InstigatorTags) "(GameplayTags=)" 1)
  (Ability|EndAbility))
